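# Masked Subsurface-shading material in the Mesh Modeling toolset, named as the output material for baked render captures, with texture parameters for each baked channel.
# A UVChannel parameter selects one of eight UV sets for all texture samples.
# BaseColor and SubsurfaceColor take tint parameters, and Emissive takes a scale.
# Per-channel static switches fall back to constants, mostly matched to buffer visualization, with a flat normal and white opacity so subsurface-only bakes stay visible.
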Material /MeshModelingToolsetExp/Materials/BakeRenderCaptureOutputSubsurfaceMaterial
graph {
 expressions {
  MaterialExpressionConstant3Vector_0 {class=MaterialExpressionConstant3Vector, properties {desc=Chosen to match the Pre Tonemap HD Color with all show flags off}}
  MaterialExpressionConstant3Vector_1 {
   class=MaterialExpressionConstant3Vector
   properties {desc="Default to a flat/up normal. Note: (Nx,Ny,Nz) = 2x(R,G,B)-(1,1,1)", constant=<rgba 0.5 0.5 1 0>}
  }
  MaterialExpressionConstant3Vector_10 {
   class=MaterialExpressionConstant3Vector
   properties {
    desc=This is used for Opacity and Opacity Mask so make it white so we see somethign ehn only SubSurface Color is baked
    constant=<rgba 1 1 1 0>
   }
  }
  MaterialExpressionConstant3Vector_2 {
   class=MaterialExpressionConstant3Vector
   properties {desc=Chosen to match the buffer visualization when nothing is passed, constant=<rgba 0.498 0.498 0.498 0>}
  }
  MaterialExpressionConstant3Vector_3 {
   class=MaterialExpressionConstant3Vector
   properties {desc=Chosen to match the buffer visualization when nothing is passed, constant=<rgba 0.7255 0.7255 0.7255 0>}
  }
  MaterialExpressionConstant3Vector_4 {class=MaterialExpressionConstant3Vector, properties {desc=Chosen to match the buffer visualization when nothing is passed}}
  MaterialExpressionConstant4Vector_0 {
   class=MaterialExpressionConstant4Vector
   inputs []
   properties {desc=Chosen to match the buffer visualization when nothing is passed, constant=<rgba 0 0 0 1>}
  }
  MaterialExpressionConstant4Vector_2 {
   class=MaterialExpressionConstant4Vector
   inputs []
   properties {desc=Chosen to match the buffer visualization when nothing is passed, constant=<rgba 0 0 0 1>}
  }
  MaterialExpressionMultiply_2 {
   class=MaterialExpressionMultiply
   inputs [
    A {from=MaterialExpressionTextureSampleParameter2D_0, output=RGB}
    B {from=MaterialExpressionVectorParameter_1, output=RGB}
   ]
  }
  MaterialExpressionMultiply_3 {
   class=MaterialExpressionMultiply
   inputs [
    A {from=MaterialExpressionVectorParameter_2, output=RGB}
    B {from=MaterialExpressionTextureSampleParameter2D_13, output=RGB}
   ]
  }
  MaterialExpressionMultiply_4 {
   class=MaterialExpressionMultiply
   inputs [
    A {from=MaterialExpressionTextureSampleParameter2D_12, output=RGB}
    B {from=MaterialExpressionScalarParameter_3}
   ]
  }
  MaterialExpressionScalarParameter_0 {class=MaterialExpressionScalarParameter, properties {parameterName=UVChannel}}
  MaterialExpressionScalarParameter_3 {class=MaterialExpressionScalarParameter, properties {defaultValue=1, parameterName=EmissiveScale}}
  MaterialExpressionStaticSwitchParameter_0 {
   class=MaterialExpressionStaticSwitchParameter
   inputs [
    True {from=MaterialExpressionMultiply_4, output=""}
    False {from=MaterialExpressionConstant3Vector_0, output=""}
   ]
   properties {parameterName=EnableEmissive}
  }
  MaterialExpressionStaticSwitchParameter_1 {
   class=MaterialExpressionStaticSwitchParameter
   inputs [
    True {from=MaterialExpressionTextureSampleParameter2D_11, output=RGB}
    False {from=MaterialExpressionConstant3Vector_1, output=""}
   ]
   properties {parameterName=EnableNormalMap}
  }
  MaterialExpressionStaticSwitchParameter_12 {
   class=MaterialExpressionStaticSwitchParameter
   inputs [
    True {from=MaterialExpressionTextureSampleParameter2D_2, output=RGB}
    False {from=MaterialExpressionConstant3Vector_10, output=""}
   ]
   properties {parameterName=EnableOpacity}
  }
  MaterialExpressionStaticSwitchParameter_13 {
   class=MaterialExpressionStaticSwitchParameter
   inputs [
    True {from=MaterialExpressionMultiply_2, output=""}
    False {from=MaterialExpressionConstant4Vector_2, output=RGBA}
   ]
   properties {parameterName=EnableSubsurfaceColor}
  }
  MaterialExpressionStaticSwitchParameter_2 {
   class=MaterialExpressionStaticSwitchParameter
   inputs [
    True {from=MaterialExpressionTextureSampleParameter2D_10, output=RGB}
    False {from=MaterialExpressionConstant3Vector_2, output=""}
   ]
   properties {parameterName=EnableRoughness}
  }
  MaterialExpressionStaticSwitchParameter_3 {
   class=MaterialExpressionStaticSwitchParameter
   inputs [
    True {from=MaterialExpressionTextureSampleParameter2D_9, output=RGB}
    False {from=MaterialExpressionConstant3Vector_3, output=""}
   ]
   properties {parameterName=EnableSpecular}
  }
  MaterialExpressionStaticSwitchParameter_4 {
   class=MaterialExpressionStaticSwitchParameter
   inputs [
    True {from=MaterialExpressionTextureSampleParameter2D_6, output=RGB}
    False {from=MaterialExpressionConstant3Vector_4, output=""}
   ]
   properties {parameterName=EnableMetallic}
  }
  MaterialExpressionStaticSwitchParameter_5 {
   class=MaterialExpressionStaticSwitchParameter
   inputs [
    True {from=MaterialExpressionMultiply_3, output=""}
    False {from=MaterialExpressionConstant4Vector_0, output=RGBA}
   ]
   properties {parameterName=EnableBaseColor}
  }
  MaterialExpressionSwitch_0 {
   class=MaterialExpressionSwitch
   inputs [
    SwitchValue {from=MaterialExpressionScalarParameter_0, output=""}
    Default {from="", output=""}
    None {from=MaterialExpressionTextureCoordinate_0, output=""}
    None {from=MaterialExpressionTextureCoordinate_9, output=""}
    None {from=MaterialExpressionTextureCoordinate_10, output=""}
    None {from=MaterialExpressionTextureCoordinate_11, output=""}
    None {from=MaterialExpressionTextureCoordinate_12, output=""}
    None {from=MaterialExpressionTextureCoordinate_13, output=""}
    None {from=MaterialExpressionTextureCoordinate_14, output=""}
    None {from=MaterialExpressionTextureCoordinate_15, output=""}
   ]
   properties {inputs [
     {
      inputName=None
      input {
       expression=&self:MaterialExpressionTextureCoordinate_0
       inputName=None
       mask=0
       maskA=0
       maskB=0
       maskG=0
       maskR=0
       outputIndex=0
      }
     }
     {
      inputName=None
      input {
       expression=&self:MaterialExpressionTextureCoordinate_9
       inputName=None
       mask=0
       maskA=0
       maskB=0
       maskG=0
       maskR=0
       outputIndex=0
      }
     }
     {
      inputName=None
      input {
       expression=&self:MaterialExpressionTextureCoordinate_10
       inputName=None
       mask=0
       maskA=0
       maskB=0
       maskG=0
       maskR=0
       outputIndex=0
      }
     }
     {
      inputName=None
      input {
       expression=&self:MaterialExpressionTextureCoordinate_11
       inputName=None
       mask=0
       maskA=0
       maskB=0
       maskG=0
       maskR=0
       outputIndex=0
      }
     }
     {
      inputName=None
      input {
       expression=&self:MaterialExpressionTextureCoordinate_12
       inputName=None
       mask=0
       maskA=0
       maskB=0
       maskG=0
       maskR=0
       outputIndex=0
      }
     }
     {
      inputName=None
      input {
       expression=&self:MaterialExpressionTextureCoordinate_13
       inputName=None
       mask=0
       maskA=0
       maskB=0
       maskG=0
       maskR=0
       outputIndex=0
      }
     }
     {
      inputName=None
      input {
       expression=&self:MaterialExpressionTextureCoordinate_14
       inputName=None
       mask=0
       maskA=0
       maskB=0
       maskG=0
       maskR=0
       outputIndex=0
      }
     }
     {
      inputName=None
      input {
       expression=&self:MaterialExpressionTextureCoordinate_15
       inputName=None
       mask=0
       maskA=0
       maskB=0
       maskG=0
       maskR=0
       outputIndex=0
      }
     }
    ]}
  }
  MaterialExpressionTextureCoordinate_0 {class=MaterialExpressionTextureCoordinate}
  MaterialExpressionTextureCoordinate_10 {class=MaterialExpressionTextureCoordinate, properties {coordinateIndex=2}}
  MaterialExpressionTextureCoordinate_11 {class=MaterialExpressionTextureCoordinate, properties {coordinateIndex=3}}
  MaterialExpressionTextureCoordinate_12 {class=MaterialExpressionTextureCoordinate, properties {coordinateIndex=4}}
  MaterialExpressionTextureCoordinate_13 {class=MaterialExpressionTextureCoordinate, properties {coordinateIndex=5}}
  MaterialExpressionTextureCoordinate_14 {class=MaterialExpressionTextureCoordinate, properties {coordinateIndex=6}}
  MaterialExpressionTextureCoordinate_15 {class=MaterialExpressionTextureCoordinate, properties {coordinateIndex=7}}
  MaterialExpressionTextureCoordinate_9 {class=MaterialExpressionTextureCoordinate, properties {coordinateIndex=1}}
  MaterialExpressionTextureSampleParameter2D_0 {
   class=MaterialExpressionTextureSampleParameter2D
   inputs [
    UVs {from=MaterialExpressionSwitch_0, output=""}
    Apply View MipBias {from="", output=""}
   ]
   properties {parameterName=SubsurfaceColor, texture=/MeshModelingToolsetExp/Textures/DefaultTexture}
  }
  MaterialExpressionTextureSampleParameter2D_10 {
   class=MaterialExpressionTextureSampleParameter2D
   inputs [
    UVs {from=MaterialExpressionSwitch_0, output=""}
    Apply View MipBias {from="", output=""}
   ]
   properties {parameterName=Roughness, texture=&DefaultTexture}
  }
  MaterialExpressionTextureSampleParameter2D_11 {
   class=MaterialExpressionTextureSampleParameter2D
   inputs [
    UVs {from=MaterialExpressionSwitch_0, output=""}
    Apply View MipBias {from="", output=""}
   ]
   properties {parameterName=NormalMap, samplerType=SAMPLERTYPE_Normal, sortPriority=-1, texture=/MeshModelingToolsetExp/Textures/DefaultNormal}
  }
  MaterialExpressionTextureSampleParameter2D_12 {
   class=MaterialExpressionTextureSampleParameter2D
   inputs [
    UVs {from=MaterialExpressionSwitch_0, output=""}
    Apply View MipBias {from="", output=""}
   ]
   properties {parameterName=Emissive, samplerType=SAMPLERTYPE_LinearColor, texture=/MeshModelingToolsetExp/Textures/DefaultEmissive}
  }
  MaterialExpressionTextureSampleParameter2D_13 {
   class=MaterialExpressionTextureSampleParameter2D
   inputs [
    UVs {from=MaterialExpressionSwitch_0, output=""}
    Apply View MipBias {from="", output=""}
   ]
   properties {parameterName=BaseColor, texture=&DefaultTexture}
  }
  MaterialExpressionTextureSampleParameter2D_2 {
   class=MaterialExpressionTextureSampleParameter2D
   inputs [
    UVs {from=MaterialExpressionSwitch_0, output=""}
    Apply View MipBias {from="", output=""}
   ]
   properties {parameterName=Opacity, texture=/MeshModelingToolsetExp/Textures/DefaultOpacity}
  }
  MaterialExpressionTextureSampleParameter2D_6 {
   class=MaterialExpressionTextureSampleParameter2D
   inputs [
    UVs {from=MaterialExpressionSwitch_0, output=""}
    Apply View MipBias {from="", output=""}
   ]
   properties {parameterName=Metallic, texture=&DefaultTexture}
  }
  MaterialExpressionTextureSampleParameter2D_9 {
   class=MaterialExpressionTextureSampleParameter2D
   inputs [
    UVs {from=MaterialExpressionSwitch_0, output=""}
    Apply View MipBias {from="", output=""}
   ]
   properties {parameterName=Specular, texture=&DefaultTexture}
  }
  MaterialExpressionVectorParameter_1 {class=MaterialExpressionVectorParameter, properties {parameterName=BrightnessSubsurface, defaultValue=<rgba 1 1 1 1>}}
  MaterialExpressionVectorParameter_2 {class=MaterialExpressionVectorParameter, properties {parameterName=Brightness, defaultValue=<rgba 1 1 1 1>}}
 }
 properties {blendMode=BLEND_Masked, pixelDepthOffsetMode=PDOM_Legacy, shadingModel=MSM_Subsurface, usedShadingModels=Subsurface}
}
