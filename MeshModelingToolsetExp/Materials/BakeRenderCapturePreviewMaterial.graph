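# Default Lit preview material for render-capture bake results in the mesh modeling toolset, with texture parameters for base color, roughness, metallic, specular, emissive and normal maps.
# A UVChannel scalar drives a Switch over eight texture coordinate sets, feeding the same selected UVs to every sampler.
# Base color is multiplied by a Brightness color parameter, and emissive by an EmissiveScale scalar.
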Material /MeshModelingToolsetExp/Materials/BakeRenderCapturePreviewMaterial
graph {
 expressions {
  MaterialExpressionMultiply_0 {
   class=MaterialExpressionMultiply
   inputs [
    A {from=MaterialExpressionVectorParameter_0, output=RGB}
    B {from=MaterialExpressionTextureSampleParameter2D_3, output=RGB}
   ]
  }
  MaterialExpressionMultiply_1 {
   class=MaterialExpressionMultiply
   inputs [
    A {from=MaterialExpressionTextureSampleParameter2D_7, output=RGB}
    B {from=MaterialExpressionScalarParameter_1}
   ]
  }
  MaterialExpressionScalarParameter_0 {class=MaterialExpressionScalarParameter, properties {parameterName=UVChannel}}
  MaterialExpressionScalarParameter_1 {class=MaterialExpressionScalarParameter, properties {defaultValue=1, parameterName=EmissiveScale}}
  MaterialExpressionSwitch_0 {
   class=MaterialExpressionSwitch
   inputs [
    SwitchValue {from=MaterialExpressionScalarParameter_0, output=""}
    Default {from="", output=""}
    None {from=MaterialExpressionTextureCoordinate_0, output=""}
    None {from=MaterialExpressionTextureCoordinate_9, output=""}
    None {from=MaterialExpressionTextureCoordinate_10, output=""}
    None {from=MaterialExpressionTextureCoordinate_11, output=""}
    None {from=MaterialExpressionTextureCoordinate_12, output=""}
    None {from=MaterialExpressionTextureCoordinate_13, output=""}
    None {from=MaterialExpressionTextureCoordinate_14, output=""}
    None {from=MaterialExpressionTextureCoordinate_15, output=""}
   ]
   properties {inputs [
     {
      inputName=None
      input {
       expression=&self:MaterialExpressionTextureCoordinate_0
       inputName=None
       mask=0
       maskA=0
       maskB=0
       maskG=0
       maskR=0
       outputIndex=0
      }
     }
     {
      inputName=None
      input {
       expression=&self:MaterialExpressionTextureCoordinate_9
       inputName=None
       mask=0
       maskA=0
       maskB=0
       maskG=0
       maskR=0
       outputIndex=0
      }
     }
     {
      inputName=None
      input {
       expression=&self:MaterialExpressionTextureCoordinate_10
       inputName=None
       mask=0
       maskA=0
       maskB=0
       maskG=0
       maskR=0
       outputIndex=0
      }
     }
     {
      inputName=None
      input {
       expression=&self:MaterialExpressionTextureCoordinate_11
       inputName=None
       mask=0
       maskA=0
       maskB=0
       maskG=0
       maskR=0
       outputIndex=0
      }
     }
     {
      inputName=None
      input {
       expression=&self:MaterialExpressionTextureCoordinate_12
       inputName=None
       mask=0
       maskA=0
       maskB=0
       maskG=0
       maskR=0
       outputIndex=0
      }
     }
     {
      inputName=None
      input {
       expression=&self:MaterialExpressionTextureCoordinate_13
       inputName=None
       mask=0
       maskA=0
       maskB=0
       maskG=0
       maskR=0
       outputIndex=0
      }
     }
     {
      inputName=None
      input {
       expression=&self:MaterialExpressionTextureCoordinate_14
       inputName=None
       mask=0
       maskA=0
       maskB=0
       maskG=0
       maskR=0
       outputIndex=0
      }
     }
     {
      inputName=None
      input {
       expression=&self:MaterialExpressionTextureCoordinate_15
       inputName=None
       mask=0
       maskA=0
       maskB=0
       maskG=0
       maskR=0
       outputIndex=0
      }
     }
    ]}
  }
  MaterialExpressionTextureCoordinate_0 {class=MaterialExpressionTextureCoordinate}
  MaterialExpressionTextureCoordinate_10 {class=MaterialExpressionTextureCoordinate, properties {coordinateIndex=2}}
  MaterialExpressionTextureCoordinate_11 {class=MaterialExpressionTextureCoordinate, properties {coordinateIndex=3}}
  MaterialExpressionTextureCoordinate_12 {class=MaterialExpressionTextureCoordinate, properties {coordinateIndex=4}}
  MaterialExpressionTextureCoordinate_13 {class=MaterialExpressionTextureCoordinate, properties {coordinateIndex=5}}
  MaterialExpressionTextureCoordinate_14 {class=MaterialExpressionTextureCoordinate, properties {coordinateIndex=6}}
  MaterialExpressionTextureCoordinate_15 {class=MaterialExpressionTextureCoordinate, properties {coordinateIndex=7}}
  MaterialExpressionTextureCoordinate_9 {class=MaterialExpressionTextureCoordinate, properties {coordinateIndex=1}}
  MaterialExpressionTextureSampleParameter2D_1 {
   class=MaterialExpressionTextureSampleParameter2D
   inputs [
    UVs {from=MaterialExpressionSwitch_0, output=""}
    Apply View MipBias {from="", output=""}
   ]
   properties {parameterName=NormalMap, samplerType=SAMPLERTYPE_Normal, sortPriority=-1, texture=/MeshModelingToolsetExp/Textures/DefaultNormal}
  }
  MaterialExpressionTextureSampleParameter2D_3 {
   class=MaterialExpressionTextureSampleParameter2D
   inputs [
    UVs {from=MaterialExpressionSwitch_0, output=""}
    Apply View MipBias {from="", output=""}
   ]
   properties {parameterName=BaseColor, texture=/MeshModelingToolsetExp/Textures/DefaultTexture}
  }
  MaterialExpressionTextureSampleParameter2D_4 {
   class=MaterialExpressionTextureSampleParameter2D
   inputs [
    UVs {from=MaterialExpressionSwitch_0, output=""}
    Apply View MipBias {from="", output=""}
   ]
   properties {parameterName=Roughness, texture=&DefaultTexture}
  }
  MaterialExpressionTextureSampleParameter2D_5 {
   class=MaterialExpressionTextureSampleParameter2D
   inputs [
    UVs {from=MaterialExpressionSwitch_0, output=""}
    Apply View MipBias {from="", output=""}
   ]
   properties {parameterName=Metallic, texture=&DefaultTexture}
  }
  MaterialExpressionTextureSampleParameter2D_7 {
   class=MaterialExpressionTextureSampleParameter2D
   inputs [
    UVs {from=MaterialExpressionSwitch_0, output=""}
    Apply View MipBias {from="", output=""}
   ]
   properties {parameterName=Emissive, samplerType=SAMPLERTYPE_LinearColor, texture=/MeshModelingToolsetExp/Textures/DefaultEmissive}
  }
  MaterialExpressionTextureSampleParameter2D_8 {
   class=MaterialExpressionTextureSampleParameter2D
   inputs [
    UVs {from=MaterialExpressionSwitch_0, output=""}
    Apply View MipBias {from="", output=""}
   ]
   properties {parameterName=Specular, texture=&DefaultTexture}
  }
  MaterialExpressionVectorParameter_0 {class=MaterialExpressionVectorParameter, properties {parameterName=Brightness, defaultValue=<rgba 1 1 1 1>}}
 }
 properties {pixelDepthOffsetMode=PDOM_Legacy, usedShadingModels=Default Lit}
}
