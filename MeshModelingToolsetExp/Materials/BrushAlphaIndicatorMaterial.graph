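# Additive, unlit, two-sided material for a modeling tool brush indicator, drawing the brush's falloff and alpha pattern without depth testing.
# It blends circular and rounded-box UV distance fields by FalloffShape, with FalloffRatio sizing the falloff band.
# FalloffMode interpolates between four falloff curves, which are multiplied by the BrushAlpha texture weighted by AlphaPower.
# IndicatorColor and IntensityScale tint and scale the result, and EyeAdaptationInverse cancels exposure.
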Material /MeshModelingToolsetExp/Materials/BrushAlphaIndicatorMaterial
graph {
 properties {
  bCastRayTracedShadows=false
  bDisableDepthTest=true
  blendMode=BLEND_Additive
  pixelDepthOffsetMode=PDOM_Legacy
  shadingModel=MSM_Unlit
  twoSided=true
  usedShadingModels=Unlit
 }
 nodes [
  n1 = VectorParameter() {parameterName=IndicatorColor, defaultValue=<rgba 1 1 1 1>}
  n2 = Subtract(A=TextureCoordinate() {desc=Distance to Circle}, B=Constant2Vector() {desc=Distance to Circle, g=0.5, r=0.5}) {desc=Distance to Circle}
  n3 = Clamp(None=OneMinus(None=ScalarParameter() {defaultValue=0.5, desc=Distance to Circle, parameterName=FalloffRatio, sliderMax=1}) {desc=Distance to Circle}) {desc=Distance to Circle}
  n4 = Subtract(
   A=Divide(A=MaterialFunctionCall(Vector 2=@n2) {desc=Distance to Circle, materialFunction=/Engine/Functions/Engine_MaterialFunctions02/Utility/VectorLength}.V2 Length) {constB=0.5, desc=Distance to Circle}
   B=@n3
  ) {desc=Distance to Circle}
  n5 = OneMinus(None=@n3) {desc=Distance to Circle}
  n6 = ScalarParameter() {defaultValue=0.5, desc=Distance to Circle, parameterName=FalloffRatio, sliderMax=1}
  n7 = Multiply(A=Clamp(None=OneMinus(None=@n6)) {desc=f}) {constB=0.5, desc=InnerRadius}
  n8 = Multiply(A=@n7) {constB=0.7071, desc=BoxHalfWidth}
  n9 = Subtract(B=@n8) {constA=0}
  n10 = MaterialFunctionCall(Float2=Subtract(A=TextureCoordinate(), B=Constant2Vector() {g=0.5, r=0.5})) {materialFunction=&BreakOutFloat2Components}
  n11 = Add(A=Max(A=Subtract(A=@n9, B=@n10.R)) {constB=0}, B=Max(A=Subtract(A=@n10.R, B=@n8)) {constB=0})
  n12 = Add(A=Max(A=Subtract(A=@n9, B=@n10.G)) {constB=0}, B=Max(A=Subtract(A=@n10.G, B=@n8)) {constB=0})
  n13 = SquareRoot(None=Add(A=Multiply(A=@n11, B=@n11), B=Multiply(A=@n12, B=@n12)) {desc=BoxDistSqr})
  n14 = Subtract(B=@n7) {constA=0.5, desc=FalloffWidth}
  n15 = LinearInterpolate(
   A=Clamp(None=Divide(A=@n4, B=@n5) {desc=Distance to Circle}) {desc=Distance to Circle}
   B=Clamp(None=Divide(A=@n13, B=@n14))
   Alpha=ScalarParameter() {desc=FalloffShape, parameterName=FalloffShape, sliderMax=1}
  )
  n16 = OneMinus(None=Power(Base=@n15))
  n17 = OneMinus(None=@n15)
  n18 = Power(Base=@n15)
  n19 = TextureCoordinate()
  n20 = TextureObjectParameter() {
   parameterName=BrushAlpha
   samplerType=SAMPLERTYPE_LinearColor
   texture=/Engine/Functions/Engine_MaterialFunctions02/ExampleContent/Textures/OffsetNoiseDistanceFields_ForNormals
  }
  n21 = Multiply(
   A=MaterialFunctionCall!(
    1=Power(Base=@n16) {constExponent=3}
    2=OneMinus(None=@n15)
    3=Power(Base=@n17) {constExponent=3}
    4=OneMinus(None=@n18)
    A=ScalarParameter() {parameterName=FalloffMode}
   ) {materialFunction=/Engine/Functions/Engine_MaterialFunctions02/Lerp_Multiple_Float}.Lerp 4 Inputs
   B=LinearInterpolate(A=Constant() {r=1}, B=TextureSample(UVs=@n19, Tex=@n20).R, Alpha=ScalarParameter() {parameterName=AlphaPower, sliderMax=1})
  )
  n22 = ScalarParameter() {defaultValue=0.2, parameterName=IntensityScale, sliderMax=1}
  EyeAdaptationInverse(LightValueInput=Multiply(A=@n1.RGB, B=Multiply(A=@n21, B=@n22))) {desc=Undo effect of exposure}
  Multiply(A=@n1.A, B=@n21)
 ]
}
