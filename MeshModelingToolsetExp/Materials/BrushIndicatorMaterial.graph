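# Unlit, two-sided translucent material for the brush indicator in the modeling tools.
# A depth fade over a falloff width scaled by object scale is remapped and clamped to tighten the gradient near intersecting surfaces.
# Its inversion, clamped to a small minimum, forms a mask that is strongest at those intersections.
# The IndicatorColor parameter passes through inverse eye adaptation to cancel exposure.
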Material /MeshModelingToolsetExp/Materials/BrushIndicatorMaterial
graph {
 properties {
  bCastRayTracedShadows=false
  blendMode=BLEND_Translucent
  pixelDepthOffsetMode=PDOM_Legacy
  shadingModel=MSM_Unlit
  twoSided=true
  usedShadingModels=Unlit
 }
 nodes [
  n1 = ScalarParameter() {defaultValue=0.6}
  n2 = Multiply(
   A=ScalarParameter() {defaultValue=0.2, parameterName=FalloffWidth}
   B=MaterialFunctionCall() {materialFunction=&ObjectScale}.Scale XYZ
  )
  n3 = Clamp(None=LinearInterpolate(
    A=ScalarParameter() {defaultValue=-1, parameterName=Param_2}
    B=ScalarParameter() {defaultValue=9, parameterName=Param_3}
    Alpha=DepthFade(Opacity=@n1, FadeDistance=@n2)
   ))
  Clamp(None=OneMinus(None=@n3)) {minDefault=0.1}
  Constant3Vector() {constant=<rgba 0.3052 0.4508 0.4472 0>}
  EyeAdaptationInverse(LightValueInput=VectorParameter() {parameterName=IndicatorColor, defaultValue=<rgba 0.1892 0.9687 1 1>}.RGB) {desc=Undo effect of exposure}
  LinearInterpolate(Alpha=@n3)
 ]
}
