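# Procedural black-and-white checker material in the Modeling Toolset, drawn over a mesh's UV coordinates.
# A UVChannel parameter switches among eight UV sets, and Density scales the chosen U and V.
# The product of two offset cosines is multiplied by a huge constant and clamped, producing hard-edged squares without a texture.
# An EyeAdaptationInverse node cancels the effect of exposure on the result.
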Material /MeshModelingToolsetExp/Materials/CheckerMaterial
graph {
 expressions {
  MaterialExpressionAdd_4 {
   class=MaterialExpressionAdd
   inputs [
    A {from=MaterialExpressionConstant_7}
    B {from=MaterialExpressionMultiply_2}
   ]
  }
  MaterialExpressionAdd_5 {
   class=MaterialExpressionAdd
   inputs [
    A {from=MaterialExpressionMultiply_3}
    B {from=MaterialExpressionConstant_8}
   ]
  }
  MaterialExpressionClamp_0 {
   class=MaterialExpressionClamp
   inputs [
    None {from=MaterialExpressionMultiply_1, output=""}
    Min {from="", output=""}
    Max {from="", output=""}
   ]
   properties {}
  }
  MaterialExpressionConstant3Vector_0 {class=MaterialExpressionConstant3Vector, properties {constant=<rgba 1 1 1 0>}}
  MaterialExpressionConstant_0 {class=MaterialExpressionConstant, properties {desc=500, r=50000}}
  MaterialExpressionConstant_7 {class=MaterialExpressionConstant, properties {r=0.5}}
  MaterialExpressionConstant_8 {class=MaterialExpressionConstant, properties {r=0.5}}
  MaterialExpressionCosine_0 {class=MaterialExpressionCosine, inputs [None {from=MaterialExpressionAdd_4, output=""}], properties {period=2}}
  MaterialExpressionCosine_1 {class=MaterialExpressionCosine, inputs [None {from=MaterialExpressionAdd_5, output=""}], properties {period=2}}
  MaterialExpressionEyeAdaptationInverse_0 {
   class=MaterialExpressionEyeAdaptationInverse
   inputs [
    LightValueInput {from=MaterialExpressionMultiply_8, output=""}
    AlphaInput {from="", output=""}
   ]
   properties {desc=Undo effect of exposure}
  }
  MaterialExpressionMaterialFunctionCall_7 {
   class=MaterialExpressionMaterialFunctionCall
   inputs [Float2 {from=MaterialExpressionSwitch_0}]
   properties {materialFunction=&BreakOutFloat2Components}
  }
  MaterialExpressionMultiply_0 {
   class=MaterialExpressionMultiply
   inputs [
    A {from=MaterialExpressionCosine_0}
    B {from=MaterialExpressionCosine_1}
   ]
  }
  MaterialExpressionMultiply_1 {
   class=MaterialExpressionMultiply
   inputs [
    A {from=MaterialExpressionMultiply_0}
    B {from=MaterialExpressionConstant_0}
   ]
  }
  MaterialExpressionMultiply_2 {
   class=MaterialExpressionMultiply
   inputs [
    A {from=MaterialExpressionMaterialFunctionCall_7, output=R}
    B {from=MaterialExpressionScalarParameter_0}
   ]
   properties {constB=100}
  }
  MaterialExpressionMultiply_3 {
   class=MaterialExpressionMultiply
   inputs [
    A {from=MaterialExpressionMaterialFunctionCall_7, output=G}
    B {from=MaterialExpressionScalarParameter_0}
   ]
  }
  MaterialExpressionMultiply_8 {
   class=MaterialExpressionMultiply
   inputs [
    A {from=MaterialExpressionConstant3Vector_0}
    B {from=MaterialExpressionClamp_0}
   ]
  }
  MaterialExpressionScalarParameter_0 {class=MaterialExpressionScalarParameter, properties {defaultValue=20, parameterName=Density, sliderMax=100, sliderMin=1}}
  MaterialExpressionScalarParameter_2 {class=MaterialExpressionScalarParameter, properties {parameterName=UVChannel}}
  MaterialExpressionSwitch_0 {
   class=MaterialExpressionSwitch
   inputs [
    SwitchValue {from=MaterialExpressionScalarParameter_2, output=""}
    Default {from="", output=""}
    None {from=MaterialExpressionTextureCoordinate_8, output=""}
    None {from=MaterialExpressionTextureCoordinate_9, output=""}
    None {from=MaterialExpressionTextureCoordinate_10, output=""}
    None {from=MaterialExpressionTextureCoordinate_11, output=""}
    None {from=MaterialExpressionTextureCoordinate_12, output=""}
    None {from=MaterialExpressionTextureCoordinate_13, output=""}
    None {from=MaterialExpressionTextureCoordinate_14, output=""}
    None {from=MaterialExpressionTextureCoordinate_15, output=""}
   ]
   properties {inputs [
     {
      inputName=None
      input {
       expression=&self:MaterialExpressionTextureCoordinate_8
       inputName=None
       mask=0
       maskA=0
       maskB=0
       maskG=0
       maskR=0
       outputIndex=0
      }
     }
     {
      inputName=None
      input {
       expression=&self:MaterialExpressionTextureCoordinate_9
       inputName=None
       mask=0
       maskA=0
       maskB=0
       maskG=0
       maskR=0
       outputIndex=0
      }
     }
     {
      inputName=None
      input {
       expression=&self:MaterialExpressionTextureCoordinate_10
       inputName=None
       mask=0
       maskA=0
       maskB=0
       maskG=0
       maskR=0
       outputIndex=0
      }
     }
     {
      inputName=None
      input {
       expression=&self:MaterialExpressionTextureCoordinate_11
       inputName=None
       mask=0
       maskA=0
       maskB=0
       maskG=0
       maskR=0
       outputIndex=0
      }
     }
     {
      inputName=None
      input {
       expression=&self:MaterialExpressionTextureCoordinate_12
       inputName=None
       mask=0
       maskA=0
       maskB=0
       maskG=0
       maskR=0
       outputIndex=0
      }
     }
     {
      inputName=None
      input {
       expression=&self:MaterialExpressionTextureCoordinate_13
       inputName=None
       mask=0
       maskA=0
       maskB=0
       maskG=0
       maskR=0
       outputIndex=0
      }
     }
     {
      inputName=None
      input {
       expression=&self:MaterialExpressionTextureCoordinate_14
       inputName=None
       mask=0
       maskA=0
       maskB=0
       maskG=0
       maskR=0
       outputIndex=0
      }
     }
     {
      inputName=None
      input {
       expression=&self:MaterialExpressionTextureCoordinate_15
       inputName=None
       mask=0
       maskA=0
       maskB=0
       maskG=0
       maskR=0
       outputIndex=0
      }
     }
    ]}
  }
  MaterialExpressionTextureCoordinate_10 {class=MaterialExpressionTextureCoordinate, properties {coordinateIndex=2}}
  MaterialExpressionTextureCoordinate_11 {class=MaterialExpressionTextureCoordinate, properties {coordinateIndex=3}}
  MaterialExpressionTextureCoordinate_12 {class=MaterialExpressionTextureCoordinate, properties {coordinateIndex=4}}
  MaterialExpressionTextureCoordinate_13 {class=MaterialExpressionTextureCoordinate, properties {coordinateIndex=5}}
  MaterialExpressionTextureCoordinate_14 {class=MaterialExpressionTextureCoordinate, properties {coordinateIndex=6}}
  MaterialExpressionTextureCoordinate_15 {class=MaterialExpressionTextureCoordinate, properties {coordinateIndex=7}}
  MaterialExpressionTextureCoordinate_8 {class=MaterialExpressionTextureCoordinate}
  MaterialExpressionTextureCoordinate_9 {class=MaterialExpressionTextureCoordinate, properties {coordinateIndex=1}}
 }
 properties {bUsedWithStaticLighting=true, pixelDepthOffsetMode=PDOM_Legacy, usedShadingModels=Default Lit}
}
