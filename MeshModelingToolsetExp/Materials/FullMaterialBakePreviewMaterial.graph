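# A default-lit preview material for full material bakes in the Mesh Modeling Toolset, judging by its name and path.
# Texture parameters for BaseColor, Metallic, Specular, Roughness, NormalMap and Emissive expose each channel as a swappable map with engine placeholder defaults.
# Metallic, Specular and Roughness sample as linear color, NormalMap samples as a normal map, and the Emissive RGB is multiplied by an EmissiveScale scalar.
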Material /MeshModelingToolsetExp/Materials/FullMaterialBakePreviewMaterial
graph {
 properties {pixelDepthOffsetMode=PDOM_Legacy, usedShadingModels=Default Lit}
 nodes [
  Multiply(
   A=TextureSampleParameter2D() {parameterName=Emissive, texture=/Engine/EngineResources/Black_Low.Black}.RGB
   B=ScalarParameter() {defaultValue=1, parameterName=EmissiveScale}
  )
  TextureSampleParameter2D() {parameterName=NormalMap, samplerType=SAMPLERTYPE_Normal, texture=/Engine/EngineMaterials/DefaultNormal}
  TextureSampleParameter2D() {parameterName=Metallic, samplerType=SAMPLERTYPE_LinearColor, texture=/Engine/ArtTools/RenderToTexture/Textures/127grey}
  TextureSampleParameter2D() {parameterName=Specular, samplerType=SAMPLERTYPE_LinearColor, texture=&127grey}
  TextureSampleParameter2D() {parameterName=BaseColor, texture=/Engine/EngineResources/DefaultTexture_Low.DefaultTexture}
  TextureSampleParameter2D() {parameterName=Roughness, samplerType=SAMPLERTYPE_LinearColor, texture=&127grey}
 ]
}
