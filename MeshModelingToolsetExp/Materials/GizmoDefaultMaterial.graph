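# Default gizmo material for the experimental Mesh Modeling Toolset, set up as a Default Lit surface.
# The graph holds a single constant red color node, with no connection to a material output shown.
# It disables depth testing and is flagged for editor compositing and static lighting use.
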Material /MeshModelingToolsetExp/Materials/GizmoDefaultMaterial
graph {
 properties {
  bDisableDepthTest=true
  bUsedWithEditorCompositing=true
  bUsedWithStaticLighting=true
  pixelDepthOffsetMode=PDOM_Legacy
  usedShadingModels=Default Lit
 }
 nodes [Constant3Vector() {constant=<rgba 0.95 0.05 0.05 0>}]
}
