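# Unlit MatCap master material in MeshModelingToolsetExp that shades meshes from an image texture indexed by surface orientation.
# A FlatShading parameter blends vertex normals toward facet normals built from world-position derivatives, disabled in orthographic views to avoid artifacts.
# The ReflectionMapping static switch derives texture UVs from either the view-space normal or a sphere-mapped reflection vector.
# The sampled color is tinted by ColorMultiplier and passed through EyeAdaptationInverse to cancel exposure.
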Material /MeshModelingToolsetExp/Materials/ImageBasedMaterial_Master
graph {
 properties {bUsedWithStaticLighting=true, pixelDepthOffsetMode=PDOM_Legacy, shadingModel=MSM_Unlit, usedShadingModels=Unlit}
 nodes [
  ComponentMask() {g=true, r=true}
  n1 = WorldPosition()
  n2 = DDX(Value=@n1.XYZ)
  n3 = DDY(Value=@n1.XYZ)
  n4 = OneMinus(None=IsOrthographic() {desc="DDX/DDY produces artifacts\r\nin Ortho views"})
  n5 = ScalarParameter() {defaultValue=1, desc=Flat Shading, parameterName=FlatShading, sliderMax=1}
  n6 = LinearInterpolate(A=VertexNormalWS(), B=Normalize(VectorInput=CrossProduct(A=@n2, B=@n3)), Alpha=Clamp(None=Multiply(A=@n4, B=@n5)))
  n7 = ReflectionVectorWS(CustomWorldNormal=@n6)
  n8 = Multiply(A=Transform(None=StaticSwitchParameter(True=@n7, False=@n6) {group=MatCap, parameterName=ReflectionMapping, sortPriority=10}) {transformSourceType=TRANSFORMSOURCE_World, transformType=TRANSFORM_View}) {constB=0.99}
  n9 = SquareRoot(None=Add(A=ComponentMask(None=@n8) {b=true}, B=Constant() {r=1}))
  n10 = Constant() {r=2.8284}
  n11 = Add(A=Divide(A=ComponentMask(None=@n8) {g=true, r=true}, B=Multiply(A=@n9, B=@n10))) {constB=0.5}
  n12 = Constant2Vector() {g=-1, r=1}
  n13 = ComponentMask(None=Add(A=@n8, B=Constant() {r=1})) {g=true, r=true}
  n14 = Constant2Vector() {g=-0.5, r=0.5}
  n15 = TextureSampleParameter2D(UVs=StaticSwitchParameter(True=Multiply(A=@n11, B=@n12) {constB=-0.5}, False=Multiply(A=@n13, B=@n14) {constB=-0.5}) {group=MatCap, parameterName=ReflectionMapping, sortPriority=10}) {
   parameterName=ImageTexture
   samplerType=SAMPLERTYPE_LinearColor
   sortPriority=-1
   texture=/MeshModelingToolsetExp/Textures/ImageTexture_Basic
  }
  EyeAdaptationInverse(LightValueInput=Multiply(A=@n15.RGB, B=VectorParameter() {parameterName=ColorMultiplier, defaultValue=<rgba 1 1 1 0>}.RGB)) {desc=Undo effect of exposure}
 ]
}
