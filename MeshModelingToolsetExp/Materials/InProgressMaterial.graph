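# InProgressMaterial is a Modeling Toolset material that draws animated two-tone blue stripes, by its name to mark work in progress.
# Local position is projected onto an approximate camera-right vector built from camera direction and world up, then offset by scaled time to scroll.
# A sine of that value is amplified and clamped into hard bands that blend DarkColor and BrightColor.
# The dimmed color passes through EyeAdaptationInverse to cancel exposure.
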Material /MeshModelingToolsetExp/Materials/InProgressMaterial
graph {
 properties {bUsedWithStaticLighting=true, pixelDepthOffsetMode=PDOM_Legacy, usedShadingModels=Default Lit}
 nodes [
  n1 = VectorParameter() {parameterName=DarkColor, defaultValue=<rgba 0.0176 0.227 0.4233 1>}
  n2 = VectorParameter() {parameterName=BrightColor, defaultValue=<rgba 0.0194 0.4969 1 1>}
  n3 = MaterialFunctionCall!(
   Vector1=MaterialFunctionCall() {materialFunction=&CameraDirectionVector}.Result
   Vector2=Constant3Vector() {constant=<rgba 0 0 1 0>}
  ) {desc="Make Camera Right Vector (approx)", materialFunction=/Engine/Functions/Engine_MaterialFunctions02/Math/CreateThirdOrthogonalVector}
  n4 = Add(
   A=DotProduct(
    A=MaterialFunctionCall() {materialFunction=&LocalPosition}.Local Position
    B=@n3.Vector3
   )
   B=Multiply(A=Time(), B=ScalarParameter() {defaultValue=10, parameterName=TimeFactor}) {constB=10}
  )
  n5 = Clamp(None=Multiply(A=Sine(None=@n4) {period=25}) {constB=100})
  EyeAdaptationInverse(LightValueInput=Multiply(A=LinearInterpolate(A=@n1.RGB, B=@n2.RGB, Alpha=@n5)) {constB=0.5}) {desc=Undo effect of exposure}
 ]
}
