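# Material function that outputs a vertex displacement vector to resolve z-fighting, with higher DepthBias inputs intended to display on top.
# In perspective views, it scales the CameraToVertex input by the bias, an empirical constant and tan half-FOV, avoiding excessive displacement at low FOV.
# In orthographic views, it multiplies the camera direction by the negated bias.
# An If on the rounded IsOrthographic value selects the approach, rounding because If rejects integers in restrictive mode.
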MaterialFunction /MeshModelingToolsetExp/Materials/GetDisplacementVectorForDepthBias
graph {nodes [
  n1 = Round(None=IsOrthographic()) {desc="We round here because IsOrthographic produces an integer value\r\n which isn't acceptable by If nodes as input under restrictive mode."}
  n2 = ViewProperty() {
   desc="Adjust for FOV such that the amount moved is the same as if the same screen area were viewed in a 90 degre FOV, otherwise we displace unnecessarily far at low FOV"
   property=MEVP_TanHalfFieldOfView
  }
  n3 = FunctionInput() {
   desc=Depth bias is used to resolve z fighting and to impose display priority on coincident items. Higher values are displayed on top.
   description="Scalar to resolve z fighting and impose display priority on coincident items (higher values are displayed on top)"
   inputName=DepthBias
   inputType=FunctionInput_Scalar
  }
  n4 = Multiply(A=@n3) {
   constB=0
   desc=We want the proportion per depth bias unit to be as small as possible while still resolving z fighting both at large distances and at obtuse angles to verts. This seems to work well empirically.
  }
  n5 = Multiply(A=FunctionInput() {inputName=CameraToVertex}, B=Multiply(A=ComponentMask(None=@n2.Property) {r=true}, B=@n4))
  n6 = MaterialFunctionCall() {materialFunction=&CameraDirectionVector}
  n7 = Multiply(A=@n3) {constB=-1}
  FunctionOutput(None=If(A=@n1, "A > B"=@n5, "A == B"=Multiply(A=@n6.Result, B=@n7), "A < B"=@n5) {constB=1, desc=Pick approach based on IsOrtho}) {description=Vector to be applied to vertices to impose the depth bias.}
 ]}
