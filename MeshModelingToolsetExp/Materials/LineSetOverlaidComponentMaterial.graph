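# Unlit translucent material for overlaid line sets that draws pixel-width lines over scene geometry with depth testing disabled.
# World position offset expands vertices using UV width and normal-encoded direction, clamps behind-camera points in perspective, and applies UV depth bias.
# Pixel depth versus scene depth detects occlusion, which OcclusionApproach renders unchanged, dashed, dithered or dimmed.
# Vertex color drives color and opacity, with exposure cancelled outside orthographic views.
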Material /MeshModelingToolsetExp/Materials/LineSetOverlaidComponentMaterial
graph {
 properties {
  bCastRayTracedShadows=false
  bDisableDepthTest=true
  blendMode=BLEND_Translucent
  pixelDepthOffsetMode=PDOM_Legacy
  shadingModel=MSM_Unlit
  twoSided=true
  usedShadingModels=Unlit
  lightmassSettings {
   bCastShadowAsMasked=false
   bOverrideCastShadowAsMasked=false
   bOverrideDiffuseBoost=false
   bOverrideEmissiveBoost=false
   bOverrideExportResolutionScale=false
   diffuseBoost=0
   emissiveBoost=1
   exportResolutionScale=1
  }
 }
 nodes [
  n1 = ComponentMask(None=TextureCoordinate()) {desc=U coordinate holds line thickness in pixels, r=true}
  n2 = ComponentMask(None=ViewSize()) {r=true}
  n3 = NamedRerouteUsage()
  n4 = DotProduct(
   A=WorldPosition() {desc=View ray, worldPositionShaderOffset=WPT_CameraRelativeNoOffsets}.XYZ
   B=MaterialFunctionCall() {materialFunction=&CameraDirectionVector}.Result
  ) {desc=Get distance to view plane passing through point}
  n5 = Multiply(
   A=ComponentMask(None=ViewProperty() {desc=Gives us ratio of half width to distance to view plane, property=MEVP_TanHalfFieldOfView}.Property) {r=true}
   B=@n4
  ) {desc=Get half width of view plane passing through point}
  n6 = MaterialFunctionCall!(Half Width) {materialFunction=/Engine/Functions/Engine_MaterialFunctions03/Viewport/MF_OrthoWidth}
  n7 = NamedRerouteUsage()
  n8 = MaterialFunctionCall() {materialFunction=&CameraDirectionVector}
  n9 = Transform(None=Constant3Vector() {constant=<rgba 0 0 1 0>}) {desc=Tangent z holds positive or negative line direction}
  n10 = MaterialFunctionCall!(
   Plane Normal=@n8.Result
   Vector=CrossProduct(A=WorldPosition() {desc=View ray, worldPositionShaderOffset=WPT_CameraRelativeNoOffsets}.XYZ, B=@n9)
  ) {materialFunction=/Engine/Functions/Engine_MaterialFunctions02/Math/ProjectVectorOntoPlane}
  n11 = CrossProduct(A=@n8.Result, B=@n9)
  n12 = Multiply(
   A=Multiply(A=Divide(A=@n1, B=@n2), B=If(A=@n3, "A > B"=@n5, "A == B"=@n6.Result, "A < B"=@n5) {constB=1}) {desc="Get world-space half thickness in a view plane \r\npassing through point"}
   B=Normalize(VectorInput=If(A=@n7, "A > B"=@n10.Result, "A == B"=@n11, "A < B"=@n10.Result) {constB=1})
  ) {desc=Move to thicken}
  n13 = Subtract(B=NamedRerouteUsage()) {desc="1 if perspective, 0 if ortho"}
  n14 = NamedRerouteUsage()
  n15 = NamedRerouteUsage()
  n16 = MaterialFunctionCall() {materialFunction=&CameraDirectionVector}
  n17 = Multiply(A=Divide(A=Multiply(A=@n14) {constB=-1}, B=DotProduct(A=@n15, B=@n16.Result) {desc=How fast line moves in normal component}) {desc=t value to intersect plane at camera}) {
   constB=0.95
   desc="For extremely long lines, the imprecision in the stored direction will result in inaccurate intersection point that is noticeable close to the camera, so don't actually move  all the way."
  }
  n18 = If(
   A=NamedRerouteUsage()
   "A > B"=@n12
   "A == B"=@n12
   "A < B"=Multiply(A=Multiply(A=@n13, B=@n17) {desc=Don't do anything if ortho}, B=NamedRerouteUsage()) {desc=Displacement to bring to plane at camera}
  ) {desc="if in front of camera, thicken, if behind, don't thicken and clamp to camera plane"}
  n19 = If(A=NamedRerouteUsage(), "A > B"=@n18, "A == B"=@n12, "A < B"=@n18) {constB=1, desc="For perspective mode, need special handling for verts behind camera plane"}
  n20 = Add(A=NamedRerouteUsage(), B=WorldPosition().XYZ)
  n21 = CameraPositionWS()
  n22 = TextureCoordinate()
  Add(
   A=@n19
   B=MaterialFunctionCall!(
    CameraToVertex=Subtract(A=@n20, B=@n21)
    DepthBias=ComponentMask(None=@n22) {desc="The V coordinate holds depth bias, which we use to move the vertex towards the camera", g=true}
   ) {materialFunction=&/GetDisplacementVectorForDepthBias}.Result
  ) {desc=Move to apply depth bias}
  n23 = VertexColor()
  Multiply(A=@n23.A, B=NamedRerouteUsage())
  n24 = EyeAdaptationInverse(LightValueInput=@n23) {desc=Undo effect of exposure}
  Multiply(
   A=If(A=NamedRerouteUsage(), "A > B"=@n24.EyeAdaptationInverse, "A == B"=@n23, "A < B"=@n24.EyeAdaptationInverse) {constB=1, desc="In ortho, EyeAdaptationInverse doesn't work for\r\ntranslucent materials"}
   B=NamedRerouteUsage()
  )
  n25 = Round(None=IsOrthographic()) {desc="We round here because IsOrthographic produces an integer value\r\n which isn't acceptable by If nodes as input under restrictive mode."}
  NamedRerouteDeclaration(None=@n25) {name=IsOrtho, nodeColor=<rgba 0.4766 1 0 1>}
  NamedRerouteDeclaration(None=@n9) {name=LineDirection, nodeColor=<rgba 0 1 0.8125 1>}
  NamedRerouteDeclaration(None=@n4) {name=DistanceToViewPlane, nodeColor=<rgba 0 1 0.0625 1>}
  NamedRerouteDeclaration(None=@n19) {name=DisplacementToCreateQuad, nodeColor=<rgba 0.1953 1 0 1>}
  NamedRerouteDeclaration(None=@n1) {name=LineWidth, nodeColor=<rgba 0.3125 1 0 1>}
  n26 = ScreenPosition()
  n27 = Constant() {desc=Sets checkerboard size, r=2}
  n28 = Floor(None=Fmod(A=Divide(A=@n26.PixelPosition, B=@n27), B=Constant() {r=2}))
  n29 = Constant()
  NamedRerouteDeclaration(None=If(A=ComponentMask(None=@n28) {r=true}, B=ComponentMask(None=@n28) {g=true}, "A > B"=@n29, "A == B"=NamedRerouteUsage(), "A < B"=@n29)) {name=DitherDim, nodeColor=<rgba 0.7812 0 1 1>}
  NamedRerouteDeclaration(None=Constant() {r=0.25}) {name=DimAmount, nodeColor=<rgba 0.5937 1 0 1>}
  n30 = MaterialFunctionCall(Vector=NamedRerouteUsage()) {
   desc="In pixel shader, need the line direction to get distance\r\nalong line for dashing. Interpolated value may have any\r\nsign, but will be colinear. Difference in signs is ok for us\r\nas long as we pass dot through symmetric function (cos).\r\nCan't do anything about 0, but that's rare."
   materialFunction=&SafeNormalize
  }
  n31 = WorldPosition()
  n32 = Cosine(None=Divide(A=DotProduct(A=@n30.Result, B=@n31.XYZ), B=NamedRerouteUsage()) {desc="Adjust period (and therefore dash length) by line\r\nwidth."}) {period=4}
  NamedRerouteDeclaration(None=Clamp(None=Multiply(A=@n32) {constB=100}) {desc=Multiplied and clamped cos is our step function}) {name=DashMultiplier, nodeColor=<rgba 1 0 0.5391 1>}
  n33 = ScalarParameter() {defaultValue=2, desc="0- no change, 1- dashed and dimmed, 2- dither and dimmed, 3- dimmed only ", parameterName=OcclusionApproach}
  n34 = Round(None=@n33)
  n35 = Constant() {r=1}
  n36 = NamedRerouteUsage()
  n37 = Constant() {r=1}
  NamedRerouteDeclaration(None=If(A=PixelDepth().R, B=SceneDepth().R, "A > B"=If(A=@n34, "A > B"=@n35, "A == B"=@n36, "A < B"=@n35) {constB=1}, "A == B"=@n37, "A < B"=@n37) {desc=1 if not occluded}) {name=OpacityMultiplier, nodeColor=<rgba 0 1 0.8828 1>}
  n38 = ScalarParameter() {defaultValue=2, desc="0- no change, 1- dashed and dimmed, 2- dither and dimmed, 3- dimmed only ", parameterName=OcclusionApproach}
  n39 = Round(None=@n38)
  n40 = NamedRerouteUsage()
  n41 = If(A=@n39, "A > B"=@n40, "A == B"=NamedRerouteUsage(), "A < B"=@n40) {constB=2}
  n42 = If(A=@n39, "A > B"=@n41, "A == B"=NamedRerouteUsage(), "A < B"=@n41) {constB=1}
  n43 = Constant() {r=1}
  n44 = Constant() {r=1}
  NamedRerouteDeclaration(None=If(A=PixelDepth().R, B=SceneDepth().R, "A > B"=If(A=@n39, "A > B"=@n42, "A == B"=@n43, "A < B"=@n42), "A == B"=@n44, "A < B"=@n44) {desc=1 if not occluded}) {name=ColorMultiplier, nodeColor=<rgba 0 1 0.8828 1>}
 ]
}
