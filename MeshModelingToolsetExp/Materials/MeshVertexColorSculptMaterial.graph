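# Unlit masked material for the modeling toolset that displays mesh vertex colors with simple view-dependent shading, named for vertex color sculpting.
# A FlatShading parameter blends pixel normals toward face normals derived from world-position DDX/DDY, disabled in orthographic views to avoid artifacts.
# The normal-to-camera dot product feeds a facing term plus a tight rim term, which multiplies vertex color.
# EyeAdaptationInverse then cancels exposure's effect on the result.
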Material /MeshModelingToolsetExp/Materials/MeshVertexColorSculptMaterial
graph {
 properties {
  bCastRayTracedShadows=false
  bUseTranslucencyVertexFog=false
  blendMode=BLEND_Masked
  pixelDepthOffsetMode=PDOM_Legacy
  shadingModel=MSM_Unlit
  translucencyLightingMode=TLM_VolumetricDirectional
  usedShadingModels=Unlit
 }
 nodes [
  Constant() {r=1}
  Constant()
  n1 = PixelNormalWS()
  n2 = WorldPosition() {desc=Flat Normal}
  n3 = Normalize(VectorInput=CrossProduct(A=DDX(Value=@n2.XYZ) {desc=Flat Normal}, B=DDY(Value=@n2.XYZ) {desc=Flat Normal}) {desc=Flat Normal}) {desc=Flat Normal}
  n4 = IsOrthographic() {desc="DDX/DDY produces artifacts\r\nin Ortho views"}
  n5 = Clamp(None=Multiply(A=OneMinus(None=@n4), B=ScalarParameter() {defaultValue=1, desc=Flat Shading, parameterName=FlatShading, sliderMax=1}))
  n6 = Clamp(None=DotProduct(A=LinearInterpolate(A=@n1, B=@n3, Alpha=@n5), B=CameraVectorWS()))
  n7 = Power(Base=OneMinus(None=@n6)) {constExponent=12}
  n8 = Power(Base=@n6) {constExponent=3}
  EyeAdaptationInverse(LightValueInput=Multiply(A=Add(A=@n7, B=@n8), B=VertexColor())) {desc=Undo effect of exposure}
 ]
}
