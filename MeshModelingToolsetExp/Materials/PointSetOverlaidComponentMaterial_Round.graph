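# Translucent unlit material that draws point-set overlay points as screen-sized round dots, visible through geometry.
# World position offset math expands vertices into camera-facing quads using tangent-Z corner directions, with UV U as size and V as depth bias toward camera.
# Interpolated tangent-Z length cuts the circle from vertex alpha.
# Vertex color dims when occluded, with exposure compensation in perspective and separate orthographic handling.
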Material /MeshModelingToolsetExp/Materials/PointSetOverlaidComponentMaterial_Round
graph {
 properties {
  bDisableDepthTest=true
  blendMode=BLEND_Translucent
  pixelDepthOffsetMode=PDOM_Legacy
  shadingModel=MSM_Unlit
  twoSided=true
  usedShadingModels=Unlit
 }
 nodes [
  n1 = Divide(A=ComponentMask(None=TextureCoordinate()) {desc=U coordinate holds point size, r=true}, B=ComponentMask(None=ViewSize()) {r=true}) {desc=Get ratio of point size to screen width}
  n2 = Custom() {code="View.ViewToClip[3][3]", description=IsOrthographic, outputType=CMOT_Float1, inputs []}
  n3 = MaterialFunctionCall() {materialFunction=&CameraDirectionVector}
  n4 = WorldPosition() {worldPositionShaderOffset=WPT_CameraRelativeNoOffsets}
  n5 = MaterialFunctionCall() {desc=Gives us ratio of half width to distance to view plane, materialFunction=/Engine/Functions/Engine_MaterialFunctions02/Utility/FOV}
  n6 = Multiply(A=DotProduct(A=Normalize(VectorInput=@n3.Result), B=@n4.XYZ) {desc=Get distance to view plane passing through point}, B=@n5."FOV ") {desc=Get half width of view plane passing through point}
  n7 = If(
   A=@n2
   "A > B"=@n6
   "A == B"=Custom() {code="1/View.ViewToClip[0][0]", description=HalfOrthoWorldScreenWidth, outputType=CMOT_Float1, inputs []}
   "A < B"=@n6
  ) {constB=1}
  n8 = Transform(None=Constant3Vector() {constant=<rgba 0 0 1 0>}) {desc=Tangent Z holds direction to move each point in camera space, transformType=TRANSFORM_Local}
  n9 = Multiply(A=@n3.Result) {constB=100, desc="We don't have a camera distance for ortho. Scale by 100 \r\nso each percent of depth offset moves 1 unit."}
  n10 = ScalarParameter() {defaultValue=1.9, parameterName=DepthOffset}
  Add(
   A=Multiply(
    A=Multiply(A=@n1, B=@n7) {desc=Get displacement amount in view plane through point}
    B=Transform(None=@n8) {transformSourceType=TRANSFORMSOURCE_Camera}
   ) {constB=2}
   B=Multiply(
    A=If(A=@n2, "A > B"=@n4.XYZ, "A == B"=@n9, "A < B"=@n4.XYZ) {constB=1, desc=Get view ray scaled by distance}
    B=Multiply(A=@n10) {constB=-0.01}
   ) {desc=Move proportion of distance toward camera}
  )
  n11 = ComponentMask(None=TextureCoordinate()) {desc=U coordinate holds point size, r=true}
  n12 = ComponentMask(None=ViewSize()) {r=true}
  n13 = Round(None=IsOrthographic()) {desc="We round here because IsOrthographic produces an integer value\r\n which isn't acceptable by If nodes as input under restrictive mode."}
  n14 = MaterialFunctionCall() {materialFunction=&CameraDirectionVector}
  n15 = MaterialFunctionCall() {desc=Gives us ratio of half width to distance to view plane, materialFunction=&FOV}
  n16 = Multiply(
   A=DotProduct(A=Normalize(VectorInput=@n14.Result), B=WorldPosition() {worldPositionShaderOffset=WPT_CameraRelativeNoOffsets}.XYZ) {desc=Get ditance to view plane passing through point}
   B=@n15."FOV "
  ) {desc=Get half width of view plane passing through point}
  n17 = MaterialFunctionCall!(Half Width) {materialFunction=/Engine/Functions/Engine_MaterialFunctions03/Viewport/MF_OrthoWidth}
  n18 = Multiply(
   A=Transform(None=Constant3Vector() {constant=<rgba 0 0 1 0>}) {desc=Tangent Z holds direction to move each point in camera space, transformType=TRANSFORM_Local}
   B=MaterialFunctionCall() {materialFunction=&ObjectScale}.Scale XYZ
  ) {desc="The local space vector we get here is inversely scaled, so undo the scaling."}
  n19 = Multiply(
   A=Multiply(
    A=Divide(A=@n11, B=@n12) {desc=Get ratio of point size to screen width}
    B=If(A=@n13, "A > B"=@n16, "A == B"=@n17.Result, "A < B"=@n16) {constB=1}
   ) {desc=Get displacement amount in view plane through point}
   B=Transform(None=@n18) {transformSourceType=TRANSFORMSOURCE_Camera}
  ) {constB=2}
  n20 = Add(A=NamedRerouteUsage(), B=WorldPosition().XYZ)
  n21 = CameraPositionWS()
  n22 = TextureCoordinate()
  Add(
   A=@n19
   B=MaterialFunctionCall!(
    CameraToVertex=Subtract(A=@n20, B=@n21)
    DepthBias=ComponentMask(None=@n22) {desc="The V coordinate holds depth bias, which we use to move the vertex towards the camera", g=true}
   ) {materialFunction=&/GetDisplacementVectorForDepthBias}.Result
  )
  n23 = VertexColor()
  EyeAdaptationInverse(LightValueInput=@n23) {desc=Undo effect of exposure}
  n24 = OneMinus(None=Clamp(None=DotProduct(A=@n8, B=@n8)))
  Multiply(A=@n23.A, B=Multiply(A=@n24, B=@n24))
  n25 = VertexColor()
  n26 = EyeAdaptationInverse(LightValueInput=@n25) {desc=Undo effect of exposure}
  n27 = Constant() {r=1}
  Multiply(
   A=If(A=@n13, "A > B"=@n26.EyeAdaptationInverse, "A == B"=@n25, "A < B"=@n26.EyeAdaptationInverse) {constB=1, desc="In ortho, EyeAdaptationInverse doesn't work for\r\ntranslucent materials"}
   B=If(A=PixelDepth().R, B=SceneDepth().R, "A > B"=Constant() {r=0.25}, "A == B"=@n27, "A < B"=@n27) {desc="0.25 if occluded, 1 otherwise"}
  ) {desc=dim if occluded}
  n28 = Constant() {r=1}
  Multiply(
   A=@n25.A
   B=If(
    A=DotProduct(A=@n18, B=@n18) {desc="In pixel shader, we can use interpolated tan z to get a distance from center (proporitional to size)"}
    "A > B"=Constant()
    "A == B"=@n28
    "A < B"=@n28
   ) {constB=1}
  )
  NamedRerouteDeclaration(None=@n19) {name=DisplacementToCreateQuad, nodeColor=<rgba 1 0 0.8203 1>}
 ]
}
