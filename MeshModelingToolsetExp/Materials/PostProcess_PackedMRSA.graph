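# Post-process material that packs Metallic, Roughness, Specular and Ambient Occlusion scene buffers into a single four-channel value.
# Four filtered SceneTexture lookups feed MakeFloat4 as X, Y, Z and A in that MRSA order, with specular taken from the stored specular buffer.
# The material is unlit and non-blendable.
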Material /MeshModelingToolsetExp/Materials/PostProcess_PackedMRSA
graph {
 properties {bIsBlendable=false, materialDomain=MD_PostProcess, pixelDepthOffsetMode=PDOM_Legacy, shadingModel=MSM_Unlit, usedShadingModels=Unlit}
 nodes [MaterialFunctionCall(
   X=SceneTexture() {bFiltered=true, sceneTextureId=PPI_Metallic}.Color
   Y=SceneTexture() {bFiltered=true, sceneTextureId=PPI_Roughness}.Color
   Z=SceneTexture() {bFiltered=true, sceneTextureId=PPI_StoredSpecular}.Color
   A=SceneTexture() {bFiltered=true, sceneTextureId=PPI_AmbientOcclusion}.Color
  ) {materialFunction=&MakeFloat4}]
}
