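# Post-process material in the Mesh Modeling Toolset, named for rendering world position as a full-screen pass.
# It is unlit, non-blendable and holds just two nodes, a SceneTexture and a WorldPosition.
# The graph shows no links between these nodes or to a material output, so how they produce the final color is not visible.
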Material /MeshModelingToolsetExp/Materials/PostProcess_WorldPosition
graph {
 properties {bIsBlendable=false, materialDomain=MD_PostProcess, pixelDepthOffsetMode=PDOM_Legacy, shadingModel=MSM_Unlit, usedShadingModels=Unlit}
 nodes [
  SceneTexture()
  WorldPosition()
 ]
}
