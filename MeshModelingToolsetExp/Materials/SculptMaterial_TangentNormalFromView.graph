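# Unlit MatCap material for mesh sculpting that shades a surface by its view-space normal, using a tangent-normal-from-view image by default.
# A static switch picks vertex normals or faceted normals built from screen derivatives of world position.
# The normal's view-space XY is remapped to UVs for the MatCap texture parameter.
# EyeAdaptationInverse then cancels exposure on the sampled color.
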Material /MeshModelingToolsetExp/Materials/SculptMaterial_TangentNormalFromView
graph {
 properties {bUsedWithStaticLighting=true, pixelDepthOffsetMode=PDOM_Legacy, shadingModel=MSM_Unlit, usedShadingModels=Unlit}
 nodes [
  n1 = WorldPosition()
  n2 = DDX(Value=@n1.XYZ)
  n3 = DDY(Value=@n1.XYZ)
  n4 = StaticSwitchParameter(True=Normalize(VectorInput=CrossProduct(A=@n2, B=@n3)), False=VertexNormalWS()) {group=MatCap, parameterName=MatCap_Faceted, sortPriority=10}
  n5 = Add(
   A=Multiply(A=Transform(None=@n4) {transformSourceType=TRANSFORMSOURCE_World, transformType=TRANSFORM_View}) {constB=0.99}
   B=Constant() {r=1}
  )
  n6 = TextureSampleParameter2D(UVs=Multiply(A=ComponentMask(None=@n5) {g=true, r=true}, B=Constant2Vector() {g=-0.5, r=0.5}) {constB=-0.5}) {
   group=MatCap
   parameterName=MatCap_TextureNormal
   sortPriority=-1
   texture=/MeshModelingToolsetExp/Textures/ImageTexture_TangentNormalFromView
  }
  EyeAdaptationInverse(LightValueInput=@n6.RGB) {desc=Undo effect of exposure}
 ]
}
