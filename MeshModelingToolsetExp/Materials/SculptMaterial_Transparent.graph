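# Translucent sculpt material from the Mesh Modeling Toolset, using Default Lit shading with surface translucency lighting.
# It exposes Color, Opacity and FresnelExponent parameters, defaulting to a blue, partly see-through look.
# FresnelExponent drives the exponent of a Fresnel node for a view-angle-dependent term.
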Material /MeshModelingToolsetExp/Materials/SculptMaterial_Transparent
graph {
 properties {blendMode=BLEND_Translucent, pixelDepthOffsetMode=PDOM_Legacy, translucencyLightingMode=TLM_Surface, usedShadingModels=Default Lit}
 nodes [
  Fresnel(ExponentIn=ScalarParameter() {defaultValue=50, parameterName=FresnelExponent})
  ScalarParameter() {defaultValue=0.65, parameterName=Opacity}
  VectorParameter() {parameterName=Color, defaultValue=<rgba 0.0606 0.309 0.842 0>}
 ]
}
