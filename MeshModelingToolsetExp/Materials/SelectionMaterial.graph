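# Two-sided unlit material for selection display in the modeling tools, tinted by a ConstantColor parameter and compensated for exposure.
# Brightness follows the view angle against a flat normal derived from world position derivatives, switching to the pixel normal in orthographic views to avoid artifacts.
# Back faces get a dimmed cosine checkerboard built from local position along approximate camera-aligned axes.
# DepthBias, summed with deprecated PercentDepthOffset, feeds a depth-bias displacement vector.
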Material /MeshModelingToolsetExp/Materials/SelectionMaterial
graph {
 properties {
  bCastRayTracedShadows=false
  bUseTranslucencyVertexFog=false
  blendMode=BLEND_Masked
  pixelDepthOffsetMode=PDOM_Legacy
  shadingModel=MSM_Unlit
  translucencyLightingMode=TLM_VolumetricDirectional
  twoSided=true
  usedShadingModels=Unlit
 }
 nodes [
  Constant() {r=1}
  Constant() {r=1}
  n1 = CameraVectorWS()
  n2 = Round(None=IsOrthographic()) {desc="We round here because IsOrthographic produces an integer value\r\n which isn't acceptable by If nodes as input under restrictive mode."}
  n3 = WorldPosition() {desc=Flat Normal}
  n4 = Normalize(VectorInput=CrossProduct(A=DDX(Value=@n3.XYZ) {desc=Flat Normal}, B=DDY(Value=@n3.XYZ) {desc=Flat Normal}) {desc=Flat Normal}) {desc=Flat Normal}
  n5 = If(A=@n2, "A > B"=@n4, "A == B"=PixelNormalWS(), "A < B"=@n4) {constB=1, desc="DDX/DDY produces artifacts in ortho\r\nviews so use input Normal then"}
  n6 = Clamp(None=Abs(None=DotProduct(A=@n1, B=@n5)))
  n7 = TwoSidedSign()
  n8 = VectorParameter() {parameterName=ConstantColor, defaultValue=<rgba 1 1 1 1>}
  n9 = MaterialFunctionCall() {materialFunction=&LocalPosition}
  n10 = MaterialFunctionCall!(
   Vector1=MaterialFunctionCall() {materialFunction=&CameraDirectionVector}.Result
   Vector2=Constant3Vector() {constant=<rgba 0 0 1 0>}
  ) {
   desc="Make Camera Up and Right Vectors (approx)"
   materialFunction=/Engine/Functions/Engine_MaterialFunctions02/Math/CreateThirdOrthogonalVector
  }
  n11 = Multiply(A=Cosine(None=DotProduct(A=@n9.Local Position, B=@n10.Vector3)) {desc=Custom period set, period=40}) {constB=100}
  n12 = Multiply(A=Cosine(None=DotProduct(A=@n9.Local Position, B=@n10.Vector2)) {period=40}) {constB=100}
  n13 = Multiply(A=@n8.RGB, B=Clamp(None=Multiply(A=@n11, B=@n12)) {desc="Dimmed as well (otherwise clamp to 0-1)", maxDefault=0.3}) {desc=Back side checkerboarded}
  EyeAdaptationInverse(LightValueInput=Multiply(A=Power(Base=@n6) {constExponent=1.5}, B=If(A=@n7, "A > B"=@n8.RGB, "A == B"=@n8.RGB, "A < B"=@n13))) {desc=Undo effect of exposure}
  MaterialFunctionCall!(
   CameraToVertex=WorldPosition() {worldPositionShaderOffset=WPT_CameraRelativeNoOffsets}.XYZ
   DepthBias=Add(
    A=ScalarParameter() {parameterName=DepthBias}
    B=ScalarParameter() {desc=DEPRECATED- use DepthBias instead, parameterName=PercentDepthOffset}
   )
  ) {materialFunction=&/GetDisplacementVectorForDepthBias}
 ]
}
