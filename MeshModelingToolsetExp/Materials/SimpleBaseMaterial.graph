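# SimpleBaseMaterial is a basic Default Lit material in MeshModelingToolsetExp, built around a single Color vector parameter.
# The parameter's RGB feeds an EyeAdaptationInverse node, which is noted as undoing the effect of exposure on the color.
# The graph shows no link from that node to a material output and contains no textures or other logic.
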Material /MeshModelingToolsetExp/Materials/SimpleBaseMaterial
graph {
 properties {pixelDepthOffsetMode=PDOM_Legacy, usedShadingModels=Default Lit}
 nodes [EyeAdaptationInverse(LightValueInput=VectorParameter() {parameterName=Color}.RGB) {desc=Undo effect of exposure}]
}
