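# Translucent, Default Lit base material for the Modeling Toolset, with Color and Opacity parameters.
# The Color is passed through EyeAdaptationInverse to undo exposure in perspective views.
# An If on the rounded IsOrthographic result uses the raw Color in orthographic views, where EyeAdaptationInverse fails for translucent materials.
# The rounding gives the If node an input it accepts under restrictive mode.
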Material /MeshModelingToolsetExp/Materials/SimpleBaseMaterial_Transparent
graph {
 properties {blendMode=BLEND_Translucent, pixelDepthOffsetMode=PDOM_Legacy, usedShadingModels=Default Lit}
 nodes [
  n1 = Round(None=IsOrthographic()) {desc="We round here because IsOrthographic produces an integer value\r\n which isn't acceptable by If nodes as input under restrictive mode."}
  n2 = VectorParameter() {parameterName=Color}
  n3 = EyeAdaptationInverse(LightValueInput=@n2.RGB) {desc=Undo effect of exposure}
  If(A=@n1, "A > B"=@n3.EyeAdaptationInverse, "A == B"=@n2.RGB, "A < B"=@n3.EyeAdaptationInverse) {constB=1, desc="In ortho, EyeAdaptationInverse doesn't work for\r\ntranslucent materials"}
  ScalarParameter() {defaultValue=1, parameterName=Opacity}
 ]
}
