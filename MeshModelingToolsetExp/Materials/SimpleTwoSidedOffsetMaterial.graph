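# Two-sided material for Mesh Modeling Toolset that offers a flat color parameter and an adjustable depth bias.
# The Color parameter passes through EyeAdaptationInverse to undo exposure.
# GetDisplacementVectorForDepthBias derives a displacement from camera-relative world position and the bias, so the depth shift comes from a displacement vector.
# The bias sums DepthBias and the deprecated PercentDepthOffset parameter.
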Material /MeshModelingToolsetExp/Materials/SimpleTwoSidedOffsetMaterial
graph {
 properties {pixelDepthOffsetMode=PDOM_Legacy, twoSided=true, usedShadingModels=Default Lit}
 nodes [
  EyeAdaptationInverse(LightValueInput=VectorParameter() {parameterName=Color}.RGB) {desc=Undo exposure setting}
  MaterialFunctionCall!(
   CameraToVertex=WorldPosition() {worldPositionShaderOffset=WPT_CameraRelativeNoOffsets}.XYZ
   DepthBias=Add(
    A=ScalarParameter() {parameterName=DepthBias}
    B=ScalarParameter() {desc=DEPRECATED- use DepthBias instead, parameterName=PercentDepthOffset}
   )
  ) {materialFunction=&/GetDisplacementVectorForDepthBias}
 ]
}
