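# Translucent material in the Modeling Toolset, named for volume editing, that shades surfaces in a fixed orange tint by how directly they face the camera.
# Perspective views derive flat normals from world-position DDX/DDY, while orthographic views use the pixel normal to avoid artifacts.
# The tinted facing term has exposure undone via EyeAdaptationInverse, except in ortho, where that fails for translucency.
# A scalar Opacity parameter is exposed.
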Material /MeshModelingToolsetExp/Materials/VolumeEditMaterial
graph {
 properties {blendMode=BLEND_Translucent, pixelDepthOffsetMode=PDOM_Legacy, usedShadingModels=Default Lit}
 nodes [
  n1 = Round(None=IsOrthographic()) {desc="We round here because IsOrthographic produces an integer value\r\n which isn't acceptable by If nodes as input under restrictive mode."}
  n2 = WorldPosition() {desc=Flat Normal}
  n3 = CrossProduct(A=DDX(Value=@n2.XYZ) {desc=Flat Normal}, B=DDY(Value=@n2.XYZ) {desc=Flat Normal}) {desc=Flat Normal}
  n4 = DotProduct(
   A=CameraVectorWS()
   B=LinearInterpolate(A=Normalize(VectorInput=@n3) {desc=Flat Normal}, B=PixelNormalWS(), Alpha=IsOrthographic()) {desc="DDX/DDY produces artifacts in ortho\r\nviews so use input Normal then"}
  )
  n5 = Multiply(A=Clamp(None=Abs(None=@n4)), B=Constant3Vector() {constant=<rgba 0.9948 0.6788 0.4041 0>})
  n6 = EyeAdaptationInverse(LightValueInput=@n5) {desc=Undo effect of exposure}
  If(A=@n1, "A > B"=@n6.EyeAdaptationInverse, "A == B"=@n5, "A < B"=@n6.EyeAdaptationInverse) {constB=1, desc="In ortho, EyeAdaptationInverse doesn't work for\r\ntranslucent materials"}
  ScalarParameter() {defaultValue=0.5, parameterName=Opacity}
 ]
}
